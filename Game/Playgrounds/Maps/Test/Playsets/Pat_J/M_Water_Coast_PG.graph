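# Masked, lit coastal water material for spline meshes in a Playgrounds test playset, with beach, ocean and waterfall looks.
# A UV1 gradient blends paired panning water normals from beach to ocean, each pair mapped in radial coordinates with its own tiling, speed and flattening.
# Panning waterfall textures, weighted by vertex-normal Z, mix into the color and the gradient-based opacity.
# That opacity feeds DitherTemporalAA for soft masked edges.
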Material /Game/Playgrounds/Maps/Test/Playsets/Pat_J/M_Water_Coast_PG
graph {
 properties {
  bUsedWithSplineMeshes=true
  blendMode=BLEND_Masked
  physMaterial=/Game/FortressPhysicalMaterials/Water
  pixelDepthOffsetMode=PDOM_Legacy
  translucencyLightingMode=TLM_Surface
  usedShadingModels=Default Lit
 }
 nodes [
  Constant()
  n1 = ScalarParameter() {defaultValue=14.5722, parameterName=Opacity}
  DepthFade(Opacity=@n1, FadeDistance=ScalarParameter() {defaultValue=2.6561, parameterName=FadeDistance})
  n2 = TextureSample(UVs=TextureCoordinate() {coordinateIndex=1}) {samplerType=SAMPLERTYPE_Alpha, texture=&/Volcano_Island_Gradient}
  n3 = Clamp(None=LinearInterpolate(A=ScalarParameter() {defaultValue=-0.7224, parameterName=C1}, B=ScalarParameter() {defaultValue=5.4392, parameterName=C2}, Alpha=@n2.A))
  LinearInterpolate(B=VectorParameter() {parameterName=Color_2, defaultValue=<rgba 0.0097 0.0319 0.04 1>}.RGB, Alpha=@n3)
  n4 = VectorParameter() {parameterName=Color_2, defaultValue=<rgba 0.0328 0.0392 0.04 1>}
  n5 = TextureCoordinate() {coordinateIndex=1}
  n6 = LinearInterpolate(
   A=ScalarParameter() {defaultValue=-0.7224, parameterName=C1}
   B=ScalarParameter() {defaultValue=5.4392, parameterName=C2}
   Alpha=TextureSample(UVs=MaterialFunctionCall(XY=@n5) {materialFunction=&Swizzle}.YX) {texture=/Game/Global/Textures/Default/Masks_and_Gradients/CylGradient}.R
  )
  n7 = TextureCoordinate()
  n8 = ScalarParameter() {defaultValue=2, parameterName=Tiling}
  n9 = Time()
  n10 = ScalarParameter() {defaultValue=-1, parameterName=Speed}
  n11 = TextureSample(UVs=Panner(Coordinate=Multiply(A=@n7, B=@n8), Time=Multiply(A=@n9, B=@n10)) {speedX=0.01, speedY=0.33}) {texture=/Game/Packages/Fortnite_UniqueMaterials_1/Textures/T_WaterFall_Tile}
  n12 = ScalarParameter() {defaultValue=0.3049, parameterName=ContrastTransition_1}
  n13 = ScalarParameter() {defaultValue=-0.7577, parameterName=ContrastTransition_2}
  n14 = VertexNormalWS()
  n15 = Multiply(A=Power(Base=ComponentMask(None=@n14) {b=true}, Exp=ScalarParameter() {defaultValue=25, parameterName=Waterfall_BlendEdge}), B=@n11.G)
  LinearInterpolate(
   A=LinearInterpolate(A=@n4.RGB, B=@n4.RGB, Alpha=Clamp(None=@n6))
   B=Multiply(A=VectorParameter() {parameterName=Color_1, defaultValue=<rgba 3 3 3 1>}.RGB, B=@n11.G)
   Alpha=Clamp(None=LinearInterpolate(A=@n12, B=@n13, Alpha=@n15))
  )
  n16 = TextureCoordinate() {uTiling=-1, vTiling=-1}
  n17 = ScalarParameter() {defaultValue=5, parameterName=Tiling Beach}
  n18 = Divide(A=@n17)
  n19 = MaterialFunctionCall(Vector or UVs=Add(A=Multiply(A=@n16, B=@n17), B=AppendVector(A=@n18, B=@n18))) {materialFunction=&VectorToRadialValue}
  n20 = Time()
  n21 = Multiply(A=@n20, B=Divide(A=ScalarParameter() {defaultValue=-1, parameterName=Speed Beach}))
  n22 = TextureSample(UVs=Panner(Coordinate=@n19.Radial Coordinates, Time=@n21) {speedX=0.01, speedY=-0.1}) {samplerType=SAMPLERTYPE_Normal, texture=/Game/Global/Textures/Default/Tileable_Patterns/T_Water2_N}
  n23 = TextureSample(UVs=Panner(Coordinate=@n19.Radial Coordinates, Time=@n21) {speedX=-0.007, speedY=-0.12}) {samplerType=SAMPLERTYPE_Normal, texture=&T_Water2_N}
  n24 = ScalarParameter() {defaultValue=3, parameterName=Tiling Ocean}
  n25 = Multiply(A=TextureCoordinate() {uTiling=-1, vTiling=-1}, B=@n24)
  n26 = Divide(A=@n24)
  n27 = MaterialFunctionCall(Vector or UVs=Add(A=@n25, B=AppendVector(A=@n26, B=@n26))) {materialFunction=&VectorToRadialValue}
  n28 = Multiply(A=@n20, B=Divide(A=ScalarParameter() {defaultValue=-0.5, parameterName=Speed ocean}, B=@n24))
  n29 = TextureSample(UVs=Panner(Coordinate=@n27.Radial Coordinates, Time=@n28) {speedX=0.01, speedY=-0.1}) {samplerType=SAMPLERTYPE_Normal, texture=&T_Water2_N}
  n30 = TextureSample(UVs=Panner(Coordinate=@n27.Radial Coordinates, Time=@n28) {speedX=-0.007, speedY=-0.12}) {samplerType=SAMPLERTYPE_Normal, texture=&T_Water2_N}
  n31 = ScalarParameter() {defaultValue=1, parameterName=FlattenNormal}
  n32 = ScalarParameter() {defaultValue=0.9, parameterName=FlattenNormal_1}
  LinearInterpolate(
   A=MaterialFunctionCall(
    Normal=MaterialFunctionCall(BaseNormal=@n22.RGB, AdditionalNormal=@n23.RGB) {materialFunction=&BlendAngleCorrectedNormals}.Result
    Flatness=ScalarParameter() {defaultValue=0.5, parameterName=beach flatness}
   ) {materialFunction=&FlattenNormal}.Result
   B=MaterialFunctionCall(
    Normal=MaterialFunctionCall(BaseNormal=@n29.RGB, AdditionalNormal=@n30.RGB) {materialFunction=&BlendAngleCorrectedNormals}.Result
    Flatness=LinearInterpolate(A=@n31, B=@n32, Alpha=@n3)
   ) {materialFunction=&FlattenNormal}.Result
   Alpha=@n3
  )
  n33 = Time()
  n34 = ScalarParameter() {defaultValue=-1, parameterName=waterfall speed}
  n35 = TextureSample(UVs=Panner(Coordinate=@n25, Time=Multiply(A=@n33, B=@n34)) {speedX=0.01, speedY=-0.33}) {texture=&T_WaterFall_Tile}
  n36 = VertexNormalWS()
  n37 = LinearInterpolate(
   A=ScalarParameter() {defaultValue=0.3049, parameterName=ContrastTransition_1}
   B=ScalarParameter() {defaultValue=-0.7577, parameterName=ContrastTransition_2}
   Alpha=Multiply(A=ComponentMask(None=@n36) {b=true}, B=@n35.G)
  )
  n38 = Power(
   Base=LinearInterpolate(A=@n2.A, B=Add(A=@n35.B, B=@n2.A), Alpha=Clamp(None=@n37))
   Exp=ScalarParameter() {defaultValue=2, parameterName=Contrast}
  )
  MaterialFunctionCall(Alpha Threshold=Clamp(None=Multiply(A=@n38, B=@n1))) {materialFunction=&DitherTemporalAA}
  Multiply(
   A=TextureSample() {texture=/Game/Packages/DS_Fortnite_TRV_Terrain/Textures/T_Transylvania_HF_Mud_1_D}.RGB
   B=VectorParameter() {parameterName=Color, defaultValue=<rgba 0.5856 0.61 0.6046 1>}.RGB
  )
  ScalarParameter() {parameterName=R}
  ScalarParameter() {defaultValue=0.1, parameterName=S}
 ]
}
